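# Environment query for the Deimos AI Director that finds nearby points with a clear line of sight, named for Creative area visibility.
# A parameter-sized grid is projected onto the navmesh, with a fallback option that repeats it using geometry-channel projection.
# Each option keeps unblocked traces from a raised context point, filters by a parameterized minimum querier distance, and scores nearer points higher.
# Test auto-sorting is off in both options.
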EnvQuery /Game/Athena/Deimos/AIDirector/EQS/Deimos_EQS_CreativeAreaVisibility
options [
 {
  generator {
   name=EnvQueryGenerator_SimpleGrid_0
   class=EnvQueryGenerator_SimpleGrid
   properties {
    bAutoSortTests=false
    spaceBetween=250
    gridSize {
     dataBinding=&self:EnvQueryGenerator_SimpleGrid_0.AIDataProvider_QueryParams_32
     dataField=FloatValue
     defaultValue=500
    }
    projectionData {
     extentX=10
     navigationFilter=&RecastFilter_UseDefaultArea
     postProjectionVerticalOffset=5
     projectDown=500
     projectUp=115
    }
   }
  }
  tests [
   EnvQueryTest_Trace_1 {
    class=EnvQueryTest_Trace
    properties {
     contextHeightOffset=30
     scoringFactor=1
     testComment=""
     testPurpose=FilterAndScore
     traceFromContext=true
     boolValue {dataBinding=null, dataField=BoolValue, defaultValue=false}
    }
   }
   EnvQueryTest_Distance_3 {
    class=EnvQueryTest_Distance
    properties {
     distanceTo=&EnvQueryContext_Querier
     filterType=Minimum
     scoringEquation=Linear
     scoringFactor=0
     testComment=""
     testMode=Distance3D
     testPurpose=Filter
     floatValueMin {
      dataBinding=&self:EnvQueryTest_Distance_3.AIDataProvider_QueryParams_34
      dataField=FloatValue
      defaultValue=0
     }
    }
   }
   EnvQueryTest_Distance_4 {
    class=EnvQueryTest_Distance
    properties {
     distanceTo=&EnvQueryContext_Querier
     filterType=Range
     floatValueMax=1500
     floatValueMin=512
     scoringEquation=InverseLinear
     scoringFactor=1
     testComment=""
     testMode=Distance3D
     testPurpose=Score
    }
   }
  ]
 }
 {
  generator {
   name=EnvQueryGenerator_SimpleGrid_2
   class=EnvQueryGenerator_SimpleGrid
   properties {
    bAutoSortTests=false
    spaceBetween=250
    gridSize {
     dataBinding=&self:EnvQueryGenerator_SimpleGrid_2.AIDataProvider_QueryParams_32
     dataField=FloatValue
     defaultValue=500
    }
    projectionData {
     extentX=10
     navigationFilter=&RecastFilter_UseDefaultArea
     postProjectionVerticalOffset=5
     projectDown=500
     projectUp=115
     traceMode=GeometryByChannel
    }
   }
  }
  tests [
   EnvQueryTest_Trace_0 {
    class=EnvQueryTest_Trace
    properties {
     contextHeightOffset=30
     scoringFactor=1
     testComment=""
     testPurpose=FilterAndScore
     traceFromContext=true
     boolValue {dataBinding=null, dataField=BoolValue, defaultValue=false}
    }
   }
   EnvQueryTest_Distance_0 {
    class=EnvQueryTest_Distance
    properties {
     distanceTo=&EnvQueryContext_Querier
     filterType=Minimum
     scoringEquation=Linear
     scoringFactor=0
     testComment=""
     testMode=Distance3D
     testPurpose=Filter
     floatValueMin {
      dataBinding=&self:EnvQueryTest_Distance_0.AIDataProvider_QueryParams_34
      dataField=FloatValue
      defaultValue=0
     }
    }
   }
   EnvQueryTest_Distance_1 {
    class=EnvQueryTest_Distance
    properties {
     distanceTo=&EnvQueryContext_Querier
     filterType=Range
     floatValueMax=1500
     floatValueMin=512
     scoringEquation=InverseLinear
     scoringFactor=1
     testComment=""
     testMode=Distance3D
     testPurpose=Score
    }
   }
  ]
 }
]
