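# EQS query for the Deimos AI Director that, per its name, finds visible edge positions in Creative.
# It generates evenly spaced circle points with a parameter-bound radius, limited by geometry traces and projected vertically.
# A trace test from the context filters and scores for unobstructed points.
# Distance to the querier filters by a parameter-bound minimum, and a second distance test linearly favors farther points.
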EnvQuery /Game/Athena/Deimos/AIDirector/EQS/Deimos_EQS_CreativeEdgeVisibility
options [{
  generator {
   name=EnvQueryGenerator_OnCircle_1
   class=EnvQueryGenerator_OnCircle
   properties {
    bIgnoreAnyContextActorsWhenGeneratingCircle=true
    circleCenterZOffset=100
    numberOfPoints=8
    pointOnCircleSpacingMethod=BySpaceBetween
    spaceBetween=250
    circleRadius {
     dataBinding=&self:EnvQueryGenerator_OnCircle_1.AIDataProvider_QueryParams_0
     dataField=FloatValue
     defaultValue=1000
    }
    projectionData {
     extentX=10
     navigationFilter=&RecastFilter_UseDefaultArea
     postProjectionVerticalOffset=5
     projectDown=500
     projectUp=115
    }
    traceData {navigationFilter=&RecastFilter_UseDefaultArea, traceMode=GeometryByChannel}
   }
  }
  tests [
   EnvQueryTest_Trace_6 {
    class=EnvQueryTest_Trace
    properties {
     contextHeightOffset=100
     scoringFactor=1
     testComment=""
     testPurpose=FilterAndScore
     traceFromContext=true
     boolValue {dataBinding=null, dataField=BoolValue, defaultValue=false}
    }
   }
   EnvQueryTest_Distance_13 {
    class=EnvQueryTest_Distance
    properties {
     distanceTo=&EnvQueryContext_Querier
     filterType=Minimum
     scoringEquation=Linear
     scoringFactor=0
     testComment=""
     testMode=Distance3D
     testPurpose=Filter
     floatValueMin {
      dataBinding=&self:EnvQueryTest_Distance_13.AIDataProvider_QueryParams_34
      dataField=FloatValue
      defaultValue=0
     }
    }
   }
   EnvQueryTest_Distance_14 {
    class=EnvQueryTest_Distance
    properties {
     distanceTo=&EnvQueryContext_Querier
     filterType=Range
     floatValueMax=1500
     floatValueMin=512
     scoringEquation=Linear
     scoringFactor=1
     testComment=""
     testMode=Distance3D
     testPurpose=Score
    }
   }
  ]
 }]
