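# Environment query for the Deimos AI Director that generates candidate rift slot locations around the querier.
# It first tries an evenly spaced ring of points, then falls back to a simple grid if the ring yields nothing.
# Both options project points onto the navmesh and only filter, without scoring, by minimum 3D distance from the querier.
# Ring radius, grid size and minimum distances are bound to query parameters, so callers can tune them.
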EnvQuery /Game/Athena/Deimos/AIDirector/EQS/Deimos_EQS_RiftSlots
options [
 {
  generator {
   name=EnvQueryGenerator_OnCircle_10
   class=EnvQueryGenerator_OnCircle
   properties {
    bAutoSortTests=false
    numberOfPoints=8
    pointOnCircleSpacingMethod=BySpaceBetween
    spaceBetween=250
    circleRadius {
     dataBinding=&self:EnvQueryGenerator_OnCircle_10.AIDataProvider_QueryParams_31
     dataField=FloatValue
     defaultValue=1000
    }
    projectionData {
     extentX=10
     navigationFilter=&RecastFilter_UseDefaultArea
     postProjectionVerticalOffset=5
     projectDown=115
     projectUp=115
    }
    traceData {navigationFilter=&RecastFilter_UseDefaultArea}
   }
  }
  tests [EnvQueryTest_Distance_0 {
    class=EnvQueryTest_Distance
    properties {
     distanceTo=&EnvQueryContext_Querier
     filterType=Minimum
     scoringEquation=Linear
     scoringFactor=0
     testComment=""
     testMode=Distance3D
     testPurpose=Filter
     floatValueMin {
      dataBinding=&self:EnvQueryTest_Distance_0.AIDataProvider_QueryParams_33
      dataField=FloatValue
      defaultValue=0
     }
    }
   }]
 }
 {
  generator {
   name=EnvQueryGenerator_SimpleGrid_12
   class=EnvQueryGenerator_SimpleGrid
   properties {
    bAutoSortTests=false
    spaceBetween=250
    gridSize {
     dataBinding=&self:EnvQueryGenerator_SimpleGrid_12.AIDataProvider_QueryParams_32
     dataField=FloatValue
     defaultValue=500
    }
    projectionData {
     extentX=10
     navigationFilter=&RecastFilter_UseDefaultArea
     postProjectionVerticalOffset=5
     projectDown=115
     projectUp=115
    }
   }
  }
  tests [EnvQueryTest_Distance_1 {
    class=EnvQueryTest_Distance
    properties {
     distanceTo=&EnvQueryContext_Querier
     filterType=Minimum
     scoringEquation=Linear
     scoringFactor=0
     testComment=""
     testMode=Distance3D
     testPurpose=Filter
     floatValueMin {
      dataBinding=&self:EnvQueryTest_Distance_1.AIDataProvider_QueryParams_34
      dataField=FloatValue
      defaultValue=0
     }
    }
   }]
 }
]
