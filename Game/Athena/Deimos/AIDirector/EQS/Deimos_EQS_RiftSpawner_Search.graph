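# An EnvQuery for the Deimos AI Director, named for searching for rift spawner candidates.
# Its single option uses a FortQueryGenerator_EncounterTargets generator to produce the encounter's targets as query items.
# It has no tests, so it adds no filtering or scoring of its own; the generator runs without async execution.
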EnvQuery /Game/Athena/Deimos/AIDirector/EQS/Deimos_EQS_RiftSpawner_Search
options [{
  generator {
   name=FortQueryGenerator_EncounterTargets_0
   class=FortQueryGenerator_EncounterTargets
   properties {bAutoSortTests=true, bCanRunAsync=false, envQueryResultNormalizationOption=Default, optionName=""}
  }
  tests []
 }]
